# Copyright Epic Games, Inc. All Rights Reserved.
#################################################
# Generated Digest of Verse API
# DO NOT modify this manually!
# Generated from build: ++Fortnite+Release-37.00-CL-44501951
#################################################


using {/Verse.org/Native}
using {/Verse.org/SceneGraph}
# Module import path: /playkigai.dig@fortnite.com/Minigame_ShuffleBoard_Game/EntityPrefabs
EntityPrefabs := module:
    @import_as("/Minigame_ShuffleBoard_Game/EntityPrefabs/sb_game.sb_game_C")
    sb_game<scoped {/playkigai.dig@fortnite.com/Minigame_ShuffleBoard_Game}> := class<final>(entity):


    sb_game_asset<scoped {/playkigai.dig@fortnite.com/Minigame_ShuffleBoard_Game}>:entity_prefab = external {}

    @import_as("/Minigame_ShuffleBoard_Game/EntityPrefabs/sb_lane.sb_lane_C")
    sb_lane<scoped {/playkigai.dig@fortnite.com/Minigame_ShuffleBoard_Game}> := class<final>(entity):


    sb_lane_asset<scoped {/playkigai.dig@fortnite.com/Minigame_ShuffleBoard_Game}>:entity_prefab = external {}

    @import_as("/Minigame_ShuffleBoard_Game/EntityPrefabs/sb_lever.sb_lever_C")
    sb_lever<scoped {/playkigai.dig@fortnite.com/Minigame_ShuffleBoard_Game}> := class<final>(entity):


    sb_lever_asset<scoped {/playkigai.dig@fortnite.com/Minigame_ShuffleBoard_Game}>:entity_prefab = external {}

    @import_as("/Minigame_ShuffleBoard_Game/EntityPrefabs/sb_puck.sb_puck_C")
    sb_puck<scoped {/playkigai.dig@fortnite.com/Minigame_ShuffleBoard_Game}> := class<final>(entity):


    sb_puck_asset<scoped {/playkigai.dig@fortnite.com/Minigame_ShuffleBoard_Game}>:entity_prefab = external {}

    @import_as("/Minigame_ShuffleBoard_Game/EntityPrefabs/sb_wall.sb_wall_C")
    sb_wall<scoped {/playkigai.dig@fortnite.com/Minigame_ShuffleBoard_Game}> := class<final>(entity):


    sb_wall_asset<scoped {/playkigai.dig@fortnite.com/Minigame_ShuffleBoard_Game}>:entity_prefab = external {}