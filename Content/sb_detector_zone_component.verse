<#
@UUID: 71c39154-1d89-43e3-b6f7-5b1e57ee895b
@Task: Detector Zone Component
@Description: Scene graph volume entity that is able to detect whether a puck is within it. Also has a points value associated with it
@Tags: Utility, SceneGraph
@Export: true
#>


using { /Fortnite.com/Devices }
using { /Verse.org/SceneGraph }
using { /Verse.org/Simulation }
using { /Fortnite.com/Game }
using { /Verse.org/Assets }

GameCategory<public><localizes>:message := "Game Settings"

sb_detector_zone_component<public> := class<final_super>(component):

    # ============ EDITABLE FIELDS ==================================

    # The amount of time in seconds after the game that this component will wait before reseting in anticipation for another
    @editable_number(int):
        MinValue := option. 0
        Categories := array{GameCategory}
    Points: int = 1

    # ============ EVENTS =======================================

    # Signals when this component's OnEndSimulation function fires; used to cancel asyncronous logic when this component stops simulating
    EndSimulationEvent<internal>: event() = event(){}
    # Signaled when a puck stops in a zone
    ZonePointUpdate<public>: event(int) = event(int){}

    # ============ INTERNAL VARS =======================================

    var PucksInZone : []sb_puck_component = array{}
    var MaybeLaneComponent: ?sb_lane_component = false  # Reference to the sibling chess_game_manager_component

    OnBeginSimulation<override>():void=
        (super:)OnBeginSimulation()

        if (RoundMan := Entity.GetFortRoundManager[]):
            RoundMan.SubscribeRoundStarted(OnRoundStart)


    OnRoundStart()<suspends>:void=
        Print("Detector Zone Round Start")

    IsPuckOverlapping(PuckComponent: sb_puck_component)<decides><transacts>:void= 
        var IsOverlapping : logic = false
        if(MeshComp := Entity.GetComponent[mesh_component]):
            for:
                OverlapHit : MeshComp.Entity.FindOverlapHits()
            do:
                if:
                    TargetComponent := OverlapHit.TargetComponent
                    PuckComponent = TargetComponent.Entity.GetParent[].GetComponent[sb_puck_component]
                then:
                    set IsOverlapping = true

        IsOverlapping?