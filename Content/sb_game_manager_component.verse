<#
@UUID: 71c39154-1d89-43e3-b6f7-5b1e57ee895b
@Task: Shuffle Board Game Manager Component
@Description: Handles the high-level logic of the game logic of the shuffle board minigame. Manages all the lanes, pucks, and interactions.
@Tags: Gameplay, SceneGraph, TeamManagement
@Export: true
#>

<#
In-Editor Entity Prefab Setup:
    (NOTE: any setting not mentioned is assumed to be default or to your preference)

sb_wall
    RootEntity:
        mesh_component (preferably a cube)
    transform_component:
        LocalTransform: scale such that it extends the entire length of the lane


sb_lever
    RootEntity:
        - sb_lever_component
            MaybeLeverRod:  set to "LeverRod" in the editor
            MaybeLeverRodPos1: set to "LeverPos1" in the editor
            MaybeLeverRodPos2:   set to "LeverPos2" in the editor
        - sb_guage_microgame_component
        LeverInteraction
            - interaction_component
            - mesh_component (preferably cylinder)
                Visible: false
            - transform_component:
                LocalTransform: scale such that is covers the entire area where the lever rod could be positioned
        LeverBase1
            - mesh_component
        LeverBaswe2
            - mesh_component
        LeverPos1:
            mesh_component (preferably a cylinder):
                Visible: false
            transform_component
                - rotate to postion that you want the lever to be in when its on
            LeverHandle:
                mesh_component (preferably sphere):
                    Visible: false
                transform_component:
                    position at end of rod
        LeverPos1:
            mesh_component (preferably a cylinder):
                Visible: false
            transform_component:
                - rotate to postion that you want the lever to be in when its off
            LeverHandle:
                mesh_component (preferably sphere):
                    Visible: false
                transform_component:
                    - position at end of rod
        LeverRod:
            mesh_component (preferably a cylinder):
            LeverHandle:
                mesh_component (preferably sphere):
                transform_component:
                    - position at end of rod
                

sb_lane:
    RootEntity:
        - sb_lane_component
            MaybeInputDevice: set to an input_trigger_device set the interact input placed in the level
            PuckSpawnPoints: populate with "PuckSpawn1", "PuckSpawn2", and "PuckSpawn3"
        LeftWall (sb_wall prefab instance)
        RightWall (sb_wall prefab instance)
        DetectorZone1Points:
            - detector_zone_component
                Points: 1
            - mesh_component (preferably a cube)
            - transform_component
                LocalTransform: position closest to the entrance of the lane, and scale to be the largest
        PointsZone1Strip:
            -mesh_component (preferably a cube)
            - transform_component
                LocalTransform: position and scale along entire width on the ground in front of DetectorZone1Points
         DetectorZone1Points:
            - detector_zone_component
                Points: 2
            - mesh_component (preferably a cube)
            - transform_component
                LocalTransform: position between DetectorZone1Points and DetectorZone4Points, and scale to be the second largest
        PointsZone2Strip:
            - mesh_component (preferably a cube)
            - transform_component
                LocalTransform: position and scale along entire width on the ground in front of DetectorZone2Points
        DetectorZone1Points:
            - detector_zone_component
                Points: 4
            - mesh_component (preferably a cube)
            - transform_component
                LocalTransform: position at the very end of the lane, and scale to be the smallest
        PointsZone2Strip:
            - mesh_component (preferably a cube)
            - transform_component
                LocalTransform: position and scale along entire width on the ground in front of DetectorZone4Points
        LeverPlatform:
            - mesh_component (preferably cylinder or cube)
            - transform_component:
                LocalTransform: scale to be a flat platform and position at entrance of lane
        Lever (instance of sb_lever entity prefab):
            - transform_component:
                LocalTransform: position on top of LeverPlatform
        PuckSpawn1
            - transoform_component:
                LocalTransform: position in front of LeverPlatform in center of lane
        PuckSpawn2
            - transoform_component:
                LocalTransform: position in front of LeverPlatform on left side of lane
        PuckSpawn3
            - transoform_component:
                LocalTransform: position in front of LeverPlatform on right side of lane
           
sb_game_template
    RootEntity
        - sb_game_manager_component
        - sb_hud_component
    [ANY NUMBER OF sb_lane prefab instances that you wish]
        
#>



using { /Verse.org }
using { /Verse.org/Concurrency }
using { /Verse.org/Simulation }
using { /Verse.org/Native }
using { /Verse.org/SceneGraph }
using { /Fortnite.com/Game }
using { /Fortnite.com/Teams }
using { /Fortnite.com/Playspaces }
using { /Fortnite.com/Devices }
using { /Verse.org/SpatialMath }
using { /Verse.org/Colors }


lane_ready_status_update_result := struct:
    LaneComp: sb_lane_component
    IsReady: logic

sb_game_manager_component<public> := class<final_super>(component):
    # ============ EDITABLE FIELDS ==================================

    # The amount of time in seconds after the game that this component will wait before reseting in anticipation for another
    @editable_number(float):
        MinValue := option. 0.0
    PostGameDelay: float = 5.0

    # ============ EVENTS =======================================
    
    # Used in AwaitAllLanesReady to cancel the async lane awaiter functions
    CancelLaneAwaitersEvent<internal>: event() = event(){}
    # Signaled when the game starts and the first turn is about to begin.
    GameBeginEvent<public>: event() = event(){}
    # Signals at the end of the game (regardless of the result); passes an optional chess_player_info object reference of the winning side (if any exists)
    GameOverEvent: event([]sb_lane_component) = event([]sb_lane_component){}
    # Signals when the game is meant to reset
    GameResetEvent<public>: event() = event(){}
    # Signals when the game manager fully reset in anticipation for the next game
    PostGameDelayCompleteEvent<public>: event() = event(){}
    # Signals when this component's OnEndSimulation function fires; used to cancel asyncronous logic when this component stops simulating
    EndSimulationEvent<internal>: event() = event(){}
    # Signals when the ready status of a lane changes; passes the lane comp that changed and and a logic indicateing that lane's new ready status
    ReadyStatusUpdateEvent: event(lane_ready_status_update_result) = event(lane_ready_status_update_result){}
    # Signals when a player has used all pucks in a lane
    LaneCompleteEvent: event(sb_lane_component) = event(sb_lane_component){}

    # ============ INTERNAL VARS =======================================

    # Map of lane components descended from this Enity on BeginSimulation. Key is the lane component, and the value is a logic that indicates whether the lane is ready to start the game or not
    var AllLaneComps: []sb_lane_component = array{}
    var MaybeHUDComponent: ?sb_hud_component = false 
    var LanesComplete : int = 0
        
    OnBeginSimulation<override>():void=
        (super:)OnBeginSimulation()
        
        # Get all Lanes Descendend from this entity
        for (Index -> LaneComp : Entity.FindDescendantComponentsArray(sb_lane_component)):
            set LaneComp.LaneID = Index
            set AllLaneComps += array{LaneComp}

        if (RoundMan := Entity.GetFortRoundManager[]):
            RoundMan.SubscribeRoundStarted(OnRoundStart)


    # Handles top-level game logic
    OnRoundStart()<suspends>:void=
        set MaybeHUDComponent = option. Entity.GetComponent[sb_hud_component]
        race:
            EndSimulationEvent.Await()
            loop:
                ResetGame()
                AwaitAllLaneCompsReady()
                GameBeginEvent.Signal()
                SpawnScoreListeners()
                SpawnPucksOnPlayerLanes()

                race:  # Primary Game Loop
                    ListenForLaneAbandonded()
                    ListenForGameComplete()

                for (LaneComp : AllLaneComps):
                    LaneComp.LockLever()

                DetermineWinners()

                # Once the game is over, displaying results is handled by the HUD Component
                Sleep(PostGameDelay)
                PostGameDelayCompleteEvent.Signal()


    OnEndSimulation<override>():void=
        (super:)OnEndSimulation()
        EndSimulationEvent.Signal()


    AwaitAllLaneCompsReady()<suspends>:void=
        race:
            EndSimulationEvent.Await()
            GameResetEvent.Await()
            block:
                for (LaneComp : AllLaneComps):
                    spawn. ListenForLaneReady(LaneComp)

                # Wait until every lane is ready
                loop:
                    ReadyStatusUpdateEvent.Await()
                    if (AllLanesReady[]) then break
                    Sleep(0.0)

        # Make sure all the async functions we spawne are ended
        CancelLaneAwaitersEvent.Signal()


    ListenForLaneAbandonded()<suspends>:void=
        race:
            EndSimulationEvent.Await()
            GameResetEvent.Await()
            block:
                for (LaneComp : AllLaneComps):
                    spawn. ListenForLaneReady(LaneComp)

                # Wait until every lane is ready
                loop:
                    ReadyStatusUpdateEvent.Await()
                    if (not AllLanesReady[]) then break
                    Sleep(0.0)

        # Make sure all the async functions we spawne are ended
        CancelLaneAwaitersEvent.Signal()


    ListenForLaneReady(LaneComponent : sb_lane_component)<suspends>:void=
        var CurrentReadyStatus: logic = logic. LaneComponent.MaybeRegisteredAgent?

        race:
            CancelLaneAwaitersEvent.Await()
            EndSimulationEvent.Await()
            loop:
                MaybeRegisteredAgent := LaneComponent.RegisteredAgentUpdateEvent.Await()
                set CurrentReadyStatus = logic. LaneComponent.MaybeRegisteredAgent?
                Result := lane_ready_status_update_result{LaneComp := LaneComponent, IsReady := CurrentReadyStatus}
                ReadyStatusUpdateEvent.Signal(Result)
                Sleep(0.0)


    SpawnPucksOnPlayerLanes():void=
        for (LaneComp : AllLaneComps):
            LaneComp.SpawnPucks()
            spawn. ListenForAllPucksUsedInLane(LaneComp)


    ListenForGameComplete()<suspends>:void = 
        set LanesComplete = 0
        race:
            EndSimulationEvent.Await()
            GameResetEvent.Await()
            loop:
                LaneCompleteEvent.Await()
                set LanesComplete += 1
                Print("Lanes Complete {LanesComplete}")
                if(LanesComplete = AllLaneComps.Length):
                    Print("Game completed")
                    return
                Sleep(0.0)
            
    
    DetermineWinners():void=
        # Get highest points
        var HighestPoints : int = 0
        for:
            LaneComp : AllLaneComps
            LaneComp.Score > HighestPoints
        do:
            set HighestPoints = LaneComp.Score

        # Get winning players
        var WinningPlayers : []sb_lane_component = array{}
        for:
            LaneComp : AllLaneComps
            LaneComp.Score = HighestPoints
        do:
            set WinningPlayers += array{LaneComp}
        
        GameOverEvent.Signal(WinningPlayers)


    SpawnScoreListeners():void=
        for(LaneComp : AllLaneComps):
            spawn. ListenForScore(LaneComp)


    ListenForScore(LaneComp : sb_lane_component)<suspends>:void=
        if (HudComp := MaybeHUDComponent?):
            race:
                EndSimulationEvent.Await()
                GameOverEvent.Await()
                loop:
                    LaneComp.ScoreUpdatedEvent.Await()
                    HudComp.UpdateScore(AllLaneComps)
                    Sleep(0.0)


    ListenForAllPucksUsedInLane(LaneComponent : sb_lane_component)<suspends>:void =
        race:
            EndSimulationEvent.Await()
            GameResetEvent.Await()
            LaneComponent.AllPucksUsedEvent.Await() 
        LaneCompleteEvent.Signal(LaneComponent)


    ResetGame():void=
        GameResetEvent.Signal()
        for (LaneComp : AllLaneComps):
            LaneComp.Reset()
            ReadyStatusUpdateEvent.Signal(lane_ready_status_update_result{LaneComp := LaneComp, IsReady := false})
    

    GetNumberLanesReady<public>()<transacts>:int =
        var NumLanesReady : int = 0
        for (LaneComp : AllLaneComps, LaneComp.MaybeRegisteredAgent?):
            set NumLanesReady += 1
        NumLanesReady
            

    AllLanesReady<public>()<decides><transacts>:void =
        Playspace := Entity.GetPlayspaceForEntity[]
        GetNumberLanesReady() >= AllLaneComps.Length
        
        