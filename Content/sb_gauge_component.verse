<#
@UUID: 71c39154-1d89-43e3-b6f7-5b1e57ee895b
@Task: Shuffle Board Minigame
@Description: Handles the bouncing guage strength timing minigame found on the puck component, The "guage" value rapidly oscilates between 0 and 100 percent and it stops when the player presses the interact button. 
@Tags: SceneGraph, UI, Gameplay, PlayerInteraction
@Export: true
#>

<#
In-Editor Entity Prefab Setup:
    (NOTE: any setting not mentioned is assumed to be default or to your preference)

sb_puck
    RootEntity:
        - sb_puck_component
        - sb_guage_microgame_component
        PuckCollision
            - interactable_component
            - mesh_component (preferably cylinder):
            - transform_component
                - LocalTransform: scale the cylinder into the shape of a large flat puck
                
#>

using { /Verse.org }
using { /Verse.org/Concurrency }
using { /Verse.org/Simulation }
using { /Verse.org/Native }
using { /Verse.org/SceneGraph }
using { /Fortnite.com/Game }
using { /Fortnite.com/Playspaces }
using { /Fortnite.com/Characters }
using { /Fortnite.com/Devices }
using { /UnrealEngine.com/Temporary/Diagnostics }


gauge_microgame_result<public>:=  struct:
    PlayerAgent<public>: agent
    MaybeFillPercent<public>: ?float


sb_gauge_microgame_component<public> := class<final_super>(component):
    # ============ EDITABLE FIELDS ==================================

    # Amount of time in seconds that the player gets to press an input before it automatically quits
    @editable_number(float):
        MinValue := option. 1.0
    TimeLimit: float = 10.0
    
    # The rate in guage percent per second that the guage oscilates
    @editable_number(float):
        MinValue := option.  0.1
        MaxValue := option. 1.0
    GuageSpeed: float = 0.5    
    
    # The percent that the guage starts at, at the start of the microgame
    @editable_number(float):
        MinValue := option. 0.0
        MaxValue := option. 1.0
    StartFillPercent: float = 0.0

    # ============ EVENTS =======================================

    # Signals when the guage minigame is initiated; passes the agent who is particpating in the microgame
    MicrogameStarted<public>: event(agent) = event(agent){}
    # Signals when the microgame is complete; passes the agent who was participating and final fill percent (if the game was successfully completed)
    MicrogameCompleted<public>: event(gauge_microgame_result) = event(gauge_microgame_result){}
    # Signals on when this component's OnEndSimulation function is called; used primarly to cancel spawned async functions
    EndSimulationEvent<public>: event() = event(){}

    # ============ INTERNAL VARS =======================================
    
    var MaybeInputDevice: ?input_trigger_device = false
    var FillPercent: float = 0.0
    var TimeRemaining: float = 0.0
    var MaybePlayerAgent: ?agent = false
    var IsFilling: logic = true


    OnBeginSimulation<override>():void=
        (super:)OnBeginSimulation()


    OnEndSimulation<override>():void=
        (super:)OnEndSimulation()
        EndSimulationEvent.Signal()
        UnregisterPlayer()
        

    PlayMicrogame(PlayerAgent:agent, InputDevice:input_trigger_device)<suspends>:?float =
        if:
            FortChar := PlayerAgent.GetFortCharacter[]
            Playspace := Entity.GetPlayspaceForEntity[]
        then: 
            RegisterPlayer(PlayerAgent)
            set TimeRemaining = TimeLimit
            set FillPercent = StartFillPercent

            MicrogameStarted.Signal(PlayerAgent)
            ReturnResult : ?float = race:
                block:  # On Registered Player Eliminated
                    FortChar.EliminatedEvent().Await()
                    false
                block:  # On End Simulation
                    EndSimulationEvent.Await()
                    false
                block:  # On Registered Player Removed From Playspace
                    loop:
                        PlayerThatLeft := Playspace.PlayerRemovedEvent().Await()
                        if (PlayerAgent = agent[PlayerThatLeft]) then break
                    false  
                block:  # On Timer Expired
                    loop:
                        set TimeRemaining -= SleepFrame()
                        if (TimeRemaining <= 0.0) then break
                    false
                block:  # On Input Pressed
                    loop:
                        Agent := InputDevice.PressedEvent.Await()
                        if (PlayerAgent = Agent ) then break
                    option. FillPercent
                block:  # Gauge Fill Logic
                    loop:
                        # The bar should flip its fill direction when it reaches the top or bottom
                        if (FillPercent >= 1.0):
                            set IsFilling = false
                        else if (FillPercent <= 0.0):
                            set IsFilling = true          
                        Print("Power Guage %: {FillPercent*100.0}")

                        Delta := SleepFrame()
                        set FillPercent += GuageSpeed * IsFilling.ToFloatSign() * Delta
                        set FillPercent = Clamp(FillPercent,0.0,1.0)
                    false
            set TimeRemaining = 0.0
            UnregisterPlayer()
            Print("Final Power Guage %: {FillPercent*100.0}")
            MicrogameCompleted.Signal(gauge_microgame_result{PlayerAgent := PlayerAgent,MaybeFillPercent := ReturnResult })
            return ReturnResult
            
        false

    
    RegisterPlayer<internal>(PlayerAgent: agent):void=
        if:
            not MaybePlayerAgent?
            FortChar := PlayerAgent.GetFortCharacter[]
            InputDevice := MaybeInputDevice?
        then:
            InputDevice.Register(PlayerAgent)
            FortChar.PutInStasis(stasis_args{AllowEmotes := false, AllowFalling := true, AllowTurning := false})
            set MaybePlayerAgent = option. PlayerAgent


    UnregisterPlayer<internal>():void=
        if:
            Player := MaybePlayerAgent?
            FortChar := Player.GetFortCharacter[]
            InputDevice := MaybeInputDevice?
        then:
            InputDevice.Unregister(Player)
            FortChar.ReleaseFromStasis()
            set MaybePlayerAgent = false

    
    

    

        



    