<#
@UUID: 71c39154-1d89-43e3-b6f7-5b1e57ee895b
@Task: Shuffle Board HUD component
@Description: Creates a HUD element on each player's screen that displays the minigame name, the ready status of each lane, the current score of each lane, and the winner at the end of the game
@Tags: SceneGraph, UI, DataStructures
@Export: true
#>

<#
In-Editor Entity Prefab Setup:
    (NOTE: any setting not mentioned is assumed to be default or to your preference)

sb_wall
    RootEntity:
        mesh_component (preferably a cube)
    transform_component:
        LocalTransform: scale such that it extends the entire length of the lane


sb_lever
    RootEntity:
        - sb_lever_component
            MaybeLeverRod:  set to "LeverRod" in the editor
            MaybeLeverRodPos1: set to "LeverPos1" in the editor
            MaybeLeverRodPos2:   set to "LeverPos2" in the editor
        - sb_guage_microgame_component
        LeverInteraction
            - interaction_component
            - mesh_component (preferably cylinder)
                Visible: false
            - transform_component:
                LocalTransform: scale such that is covers the entire area where the lever rod could be positioned
        LeverBase1
            - mesh_component
        LeverBaswe2
            - mesh_component
        LeverPos1:
            mesh_component (preferably a cylinder):
                Visible: false
            transform_component
                - rotate to postion that you want the lever to be in when its on
            LeverHandle:
                mesh_component (preferably sphere):
                    Visible: false
                transform_component:
                    position at end of rod
        LeverPos1:
            mesh_component (preferably a cylinder):
                Visible: false
            transform_component:
                - rotate to postion that you want the lever to be in when its off
            LeverHandle:
                mesh_component (preferably sphere):
                    Visible: false
                transform_component:
                    - position at end of rod
        LeverRod:
            mesh_component (preferably a cylinder):
            LeverHandle:
                mesh_component (preferably sphere):
                transform_component:
                    - position at end of rod
                

sb_lane:
    RootEntity:
        - sb_lane_component
            MaybeInputDevice: set to an input_trigger_device set the interact input placed in the level
            PuckSpawnPoints: populate with "PuckSpawn1", "PuckSpawn2", and "PuckSpawn3"
        LeftWall (sb_wall prefab instance)
        RightWall (sb_wall prefab instance)
        DetectorZone1Points:
            - detector_zone_component
                Points: 1
            - mesh_component (preferably a cube)
            - transform_component
                LocalTransform: position closest to the entrance of the lane, and scale to be the largest
        PointsZone1Strip:
            -mesh_component (preferably a cube)
            - transform_component
                LocalTransform: position and scale along entire width on the ground in front of DetectorZone1Points
         DetectorZone1Points:
            - detector_zone_component
                Points: 2
            - mesh_component (preferably a cube)
            - transform_component
                LocalTransform: position between DetectorZone1Points and DetectorZone4Points, and scale to be the second largest
        PointsZone2Strip:
            - mesh_component (preferably a cube)
            - transform_component
                LocalTransform: position and scale along entire width on the ground in front of DetectorZone2Points
        DetectorZone1Points:
            - detector_zone_component
                Points: 4
            - mesh_component (preferably a cube)
            - transform_component
                LocalTransform: position at the very end of the lane, and scale to be the smallest
        PointsZone2Strip:
            - mesh_component (preferably a cube)
            - transform_component
                LocalTransform: position and scale along entire width on the ground in front of DetectorZone4Points
        LeverPlatform:
            - mesh_component (preferably cylinder or cube)
            - transform_component:
                LocalTransform: scale to be a flat platform and position at entrance of lane
        Lever (instance of sb_lever entity prefab):
            - transform_component:
                LocalTransform: position on top of LeverPlatform
        PuckSpawn1
            - transoform_component:
                LocalTransform: position in front of LeverPlatform in center of lane
        PuckSpawn2
            - transoform_component:
                LocalTransform: position in front of LeverPlatform on left side of lane
        PuckSpawn3
            - transoform_component:
                LocalTransform: position in front of LeverPlatform on right side of lane
            
#>



using { /UnrealEngine.com/Temporary/UI }
using { /UnrealEngine.com/Temporary/SpatialMath }
using { /Fortnite.com/UI }
using { /Fortnite.com/Devices }
using { /Fortnite.com/Teams }
using { /Fortnite.com/Playspaces }
using { /Verse.org }    
using { /Verse.org/Colors }
using { /Verse.org/Native }
using { /Verse.org/SceneGraph }
using { /Verse.org/Simulation }
using { /Fortnite.com/Game }


###################### Start UI Component ##########################################################################

sb_hud_component<public> := class<final_super>(component):

    var PlayerToUI:[player]sb_hud = map{} # Stores the UI object for each player
    var MaybeGameManager: ?sb_game_manager_component = false  # Reference to the sibling chess_game_manager_component
    
    OnBeginSimulation<override>():void =
        (super:)OnBeginSimulation()

        if (RoundMan := Entity.GetFortRoundManager[]):
            RoundMan.SubscribeRoundStarted(OnRoundStart)

        # Get Reference to Sibling Game Manager Component and setup gameplay callbacks
        set MaybeGameManager = option. Entity.GetComponent[sb_game_manager_component] 

        if (Playspace := Entity.GetPlayspaceForEntity[]):
            Playspace.PlayerRemovedEvent().Subscribe(OnPlayerRemoved)
        

    OnRoundStart()<suspends>:void=
        Sleep(1.0)
        if (GameManager := MaybeGameManager?):
            AddUIAllPlayers(GameManager.AllLaneComps)
            loop:
                # Waiting For Player Ready
                race:
                    GameManager.GameBeginEvent.Await()
                    loop: # On Player Ready
                        UpdateReadyStatus(GameManager.ReadyStatusUpdateEvent.Await())
                        Sleep(0.0)

                for (Player->Hud : PlayerToUI):
                    Hud.ResetGameHUDToMatchStart()

                # Game Loop
                Winners := GameManager.GameOverEvent.Await()

                # Post Game
                # If playing with friends...
                if(GameManager.AllLaneComps.Length > 1):
                    UpdateResultText(StringToMessage(GetWinnersString(Winners)))
                # If playing alone, display a high score
                else if (Winner := Winners[0]):
                    UpdateResultText(StringToMessage("Score: " + ToString(Winner.Score)))
                
                GameManager.GameResetEvent.Await()
                ResetUI()
                
                    
    OnEndSimulation<override>():void =
        RemoveUIAllPlayers()    # Remove all UI elements on end; Custom Verse UI does not get cleaned up automatically.


    GetWinnersString(Winners : []sb_lane_component):string=
        var FinalWinMsg : string = "GetWinnerMessage Failed"

        if (Winners.Length > 1):
            set FinalWinMsg = "The winners are:"

            for:
                Index -> Winner : Winners
                WinningLaneNumber := Winner.LaneID
            do:
                set FinalWinMsg += " Player " + ToString(WinningLaneNumber)

                if(Winners.Length = 2 and Index = 0):
                    set FinalWinMsg += " and"

                if(Winners.Length > 2):
                    if(Index < Winners.Length - 1):
                        set FinalWinMsg += ","

                        if(Index < Winners.Length - 2):
                            set FinalWinMsg += " and"

        else if (WinningLaneNumber := Winners[0].LaneID):
            set FinalWinMsg = "The winner is Player {WinningLaneNumber+1}"

        FinalWinMsg

    
    UpdateReadyStatus(UpdateResult : lane_ready_status_update_result) : void = 

        for (Player -> Hud : PlayerToUI):
            if (Textbox := Hud.PointsTextBlocks[UpdateResult.LaneComp.LaneID]):

                var ScoreString : string = ""
                Textbox.SetTextColor(NamedColors.Red)

                if (UpdateResult.IsReady?):
                    set ScoreString = "Ready"
                    Textbox.SetTextColor(Hud.ReadyColor)
                
                Textbox.SetText(StringToMessage(ScoreString))


    UpdateScore(PlayerInfoArray : []sb_lane_component):void =
        for:
            Player -> Hud : PlayerToUI
            Index -> PlayerInfo : PlayerInfoArray
            Textbox := Hud.PointsTextBlocks[Index]
        do:
            ScoreMsg := StringToMessage(ToString(PlayerInfo.Score))
            Textbox.SetText(ScoreMsg)
            Textbox.SetTextColor(Hud.ScoreColor)
                

    UpdateResultText(Message:message):void = 
        for (Player->Hud : PlayerToUI):
            Hud.ResultText.SetText(Message)

            
    ResetUI():void = 
        for (Player->Hud : PlayerToUI):
            Hud.ResetGameHUDToUnready()


    AddUIAllPlayers(LaneComps : []sb_lane_component):void =
        if (Playspace := Entity.GetPlayspaceForEntity[]):
            for (Player : Playspace.GetPlayers()) do Player.AddUI(LaneComps)


    RemoveUIAllPlayers():void=
        if(Playspace := Entity.GetPlayspaceForEntity[]):
            for(Player : Playspace.GetPlayers()):
                Player.RemoveUI()
            set PlayerToUI = map{}
            

    (Player : player).AddUI(LanesInfoArray : []sb_lane_component):void =
        if:
            PlayerUI := GetPlayerUI[Player]
            ShuffleboardUI := sb_hud{Lanes := LanesInfoArray}
            set PlayerToUI[Player] = ShuffleboardUI
            GameManager := MaybeGameManager?
        then:
            ShuffleboardUI.CreateUI()
            PlayerUI.AddWidget(ShuffleboardUI.GetWidget())
        

    (Player : player).RemoveUI():void =
        if:
            PlayerUI := GetPlayerUI[Player]
            ShuffleboardUI := PlayerToUI[Player]
        then:
            PlayerUI.RemoveWidget(ShuffleboardUI.GetWidget())
            set PlayerToUI = PlayerToUI.RemoveKey(Player)
    

    OnPlayerAdded(Player: player):void = Player.AddUI()
    OnPlayerRemoved(Player: player):void = Player.RemoveUI()

###################### End UI Component #############################################################################


###################### Start UI Display Messages ######################################################################

TitleMessage<localizes>() : message = "Shuffleboard"
DefaultPlayerMessage<localizes>() : message = ""
DefaultPointsText<localizes>() : message = "0"
WaitingOnPlayer1<localizes>() : message = "Waiting for Player 1 to ready up"
WaitingOnPlayer2<localizes>() : message = "Waiting for Player 2 to ready up"
WaitingOnPlayers<localizes>() : message = "Waiting for both players to be ready"
GameStartMessage<localizes>(String: string) : message = "Game Starting in {String}"
PlayerWinMessage<localizes>(String: string) : message = "{String} is the victor!"

###################### End UI Display Messages ##########################################################################


###################### Start Player UI Class ############################################################################
                
sb_hud := class:
    Lanes : []sb_lane_component = array{}
    ScoreColor : color = NamedColors.White
    ReadyColor : color = NamedColors.LightGreen
    ResultText : text_block = text_block{DefaultTextColor := NamedColors.Yellow}
    var MainWidget : canvas = canvas{}
    var PointsTextBlocks : []text_block = array{}
    var PointsStackboxSlots : []stack_box_slot = array{}

    GetWidget() : canvas = MainWidget

    ResetGameHUDToUnready() : void = 
        ResultText.SetText(WaitingOnPlayers())
        for(PointTextBlock : PointsTextBlocks):
            PointTextBlock.SetText(DefaultPlayerMessage())


    ResetGameHUDToMatchStart() : void = 
        ResultText.SetText(StringToMessage(""))

        for(PointTextBlock : PointsTextBlocks):
            PointTextBlock.SetTextColor(ScoreColor)
            PointTextBlock.SetText(DefaultPointsText())


    CreateNewBlackBoxWidget():overlay_slot =
        overlay_slot{
            Widget := color_block:
                DefaultColor := NamedColors.Black
                DefaultOpacity := 0.65
                DefaultDesiredSize := vector2{X := 64.0, Y := 64.0}
            }


    CreateColorBoxWidget(Color : color):overlay_slot =
        overlay_slot{
            Widget := color_block:
                DefaultColor := Color
                DefaultOpacity := 0.65
                DefaultDesiredSize := vector2{X := 72.0, Y := 72.0}
        }


    CreateNewPointTextWidget(Color : color):text_block =
        Widget := text_block:
            DefaultText := StringToMessage("")
            DefaultTextColor := Color
            DefaultJustification := text_justification.Center


    CreateNewStackboxSlotPointWidget(BlackBoxOverlaySlot : overlay_slot, ColorBoxOverlaySlot : overlay_slot, PointTextBoxOverlaySlot : overlay_slot):stack_box_slot =
        stack_box_slot{
            Padding := margin{Left := 8.0, Top := 8.0, Right := 8.0, Bottom := 8.0}
            Widget := overlay{
                
                Slots := array{
                    BlackBoxOverlaySlot,
                    ColorBoxOverlaySlot,
                    PointTextBoxOverlaySlot
                }
            }
        }


    CreateUI() : void =
        GameTitle: text_block = text_block{DefaultTextColor := NamedColors.White, DefaultText := TitleMessage()},

        for(Index := 0..Lanes.Length-1):
            if:
                LaneComponent := Lanes[Index]
                LaneColor := LaneComponent.LaneColor
            then:
                set PointsTextBlocks += array{NewTextBlock := CreateNewPointTextWidget(NamedColors.White)}

                BlackBoxOverlaySlot := CreateNewBlackBoxWidget()
                ColorBoxOverlaySlot := CreateColorBoxWidget(LaneColor)
                TextBlockOverlaySlot := overlay_slot {Widget := NewTextBlock}
                set PointsStackboxSlots += array{CreateNewStackboxSlotPointWidget(BlackBoxOverlaySlot, ColorBoxOverlaySlot, TextBlockOverlaySlot)}


        ScoreUI := canvas:
            Slots := array:

                # Create title at the top of the screen
                canvas_slot:
                    Anchors := anchors{Minimum := vector2{X:= 0.5, Y := 0.0}, Maximum := vector2{X:= 0.5, Y := 0.0}}
                    Offsets := margin{Left := 0.0, Top := 130.0, Right := 650.0, Bottom := 100.0}
                    Alignment := vector2{X := 0.5, Y := 0.5}
                    Widget := overlay{
                        Slots := array{
                            overlay_slot:
                                Widget := color_block:
                                    DefaultColor := MakeColorFromSRGB(1.0, 0.07, 0.07)
                                    DefaultOpacity := 0.5
                                    DefaultDesiredSize := vector2{X := 400.0, Y := 80.0}  

                            overlay_slot{
                                Widget := GameTitle
                            }
                        }
                    }

                # Create point boxes
                canvas_slot:
                    Anchors := anchors{Minimum := vector2{X:= 0.5, Y := 0.0}, Maximum := vector2{X:= 0.5, Y := 0.0}}
                    Offsets := margin{Left := 0.0, Top := 220.0, Right := 1000.0, Bottom := 100.0}
                    Alignment := vector2{X := 0.5, Y := 0.5}
                    Widget := stack_box{
                        Orientation := orientation.Horizontal,
                        Slots := PointsStackboxSlots
                    }
                    
        
        #Adds a timer and winner/draw text underneath the score text above
        OverallUI := stack_box{
            Orientation := orientation.Vertical,
            Slots := array{
                stack_box_slot{
                    Widget := ScoreUI,
                    HorizontalAlignment := horizontal_alignment.Center,
                    VerticalAlignment := vertical_alignment.Center
                },
                stack_box_slot{
                    Widget := ResultText,
                    HorizontalAlignment := horizontal_alignment.Center,
                    VerticalAlignment := vertical_alignment.Center
                }
            }
        }

        set MainWidget = canvas{
            Slots := array{
                canvas_slot{
                    Widget := OverallUI,
                    Anchors := anchors{Minimum := vector2{X := 0.5, Y := 0.1}, Maximum := vector2{X := 0.5, Y := 0.1}},
                    Alignment := vector2{X := 0.5, Y := 0.5}
                }
            }
        }

        ResetGameHUDToUnready()

###################### Puck Fling Bar HUD Class ############################################################################
                
sb_hud_puck_bar := class:

    BarLength : int = 10
    ResultText : text_block = text_block{DefaultTextColor := NamedColors.Yellow}

    var MainWidget : canvas = canvas{}

    var PuckWidgets : []overlay_slot = array{}
    var PuckStackboxSlots : []stack_box_slot = array{}

    GetWidget() : canvas = MainWidget


    ResetPuckBar() : void = 
        ResultText.SetText(WaitingOnPlayers())

        for (Index -> PointWidget : PuckWidgets):
            PointWidget.Widget.SetVisibility(widget_visibility.Hidden)


    CreateNewPointWidget(Color : color):overlay_slot =
        overlay_slot{
            Widget := color_block:
                DefaultColor := Color
                DefaultOpacity := 0.65
                DefaultDesiredSize := vector2{X := 64.0, Y := 64.0}
            }


    CreateNewStackboxSlotPointWidget(PointOverlaySlot : overlay_slot):stack_box_slot =
        stack_box_slot{
            Padding := margin{Left := 8.0, Top := 8.0, Right := 8.0, Bottom := 8.0}
            Widget := overlay{
                
                Slots := array{
                    overlay_slot{
                        Widget := color_block:
                            DefaultColor := NamedColors.Black
                            DefaultOpacity := 0.65
                            DefaultDesiredSize := vector2{X := 72.0, Y := 72.0}
                    },

                    PointOverlaySlot
                }
            }
        }


    CreateUI() : void =
        GameTitle: text_block = text_block{DefaultTextColor := NamedColors.White, DefaultText := TitleMessage()},

        for(Index := 0..BarLength-1):
            set PuckWidgets += array{CreateNewPointWidget(NamedColors.Yellow)}

            if(PointOverlaySlot := PuckWidgets[Index]):
                set PuckStackboxSlots += array{CreateNewStackboxSlotPointWidget(PointOverlaySlot)}


        ScoreUI := canvas:
            Slots := array:
                # Create boxes
                canvas_slot:
                    Anchors := anchors{Minimum := vector2{X:= 0.5, Y := 0.0}, Maximum := vector2{X:= 0.5, Y := 0.0}}
                    Offsets := margin{Left := 0.0, Top := 220.0, Right := 1000.0, Bottom := 100.0}
                    Alignment := vector2{X := 0.5, Y := 0.5}
                    Widget := stack_box{
                        Orientation := orientation.Horizontal,
                        Slots := array{
                            stack_box_slot{
                                HorizontalAlignment := horizontal_alignment.Right
                                Widget := stack_box{
                                Orientation := orientation.Horizontal,
                                Slots := PuckStackboxSlots
                                }
                            }
                        }
                    }
        
        #Adds a timer and winner/draw text underneath the score text above
        OverallUI := stack_box{
            Orientation := orientation.Vertical,
            Slots := array{
                stack_box_slot{
                    Widget := ScoreUI,
                    HorizontalAlignment := horizontal_alignment.Center,
                    VerticalAlignment := vertical_alignment.Center
                },
                stack_box_slot{
                    Widget := ResultText,
                    HorizontalAlignment := horizontal_alignment.Center,
                    VerticalAlignment := vertical_alignment.Center
                }
            }
        }

        set MainWidget = canvas{
            Slots := array{
                canvas_slot{
                    Widget := OverallUI,
                    Anchors := anchors{Minimum := vector2{X := 0.5, Y := 0.1}, Maximum := vector2{X := 0.5, Y := 0.1}},
                    Alignment := vector2{X := 0.5, Y := 0.5}
                }
            }
        }

        ResetPuckBar()