<#
@UUID: 71c39154-1d89-43e3-b6f7-5b1e57ee895b
@Task: Shuffle Board Lane Component
@Description: Component that manages a sb_lane prefab. Handles player registration, puck spawning, and scoring of an individual lane
@Tags: SceneGraph, Gameplay, PlayerManagement
@Export: true
#>

<#
In-Editor Entity Prefab Setup:
    (NOTE: any setting not mentioned is assumed to be default or to your preference)

sb_wall
    RootEntity:
        mesh_component (preferably a cube)
    transform_component:
        LocalTransform: scale such that it extends the entire length of the lane


sb_lever
    RootEntity:
        - sb_lever_component
            MaybeLeverRod:  set to "LeverRod" in the editor
            MaybeLeverRodPos1: set to "LeverPos1" in the editor
            MaybeLeverRodPos2:   set to "LeverPos2" in the editor
        - sb_guage_microgame_component
        LeverInteraction
            - interaction_component
            - mesh_component (preferably cylinder)
                Visible: false
            - transform_component:
                LocalTransform: scale such that is covers the entire area where the lever rod could be positioned
        LeverBase1
            - mesh_component
        LeverBaswe2
            - mesh_component
        LeverPos1:
            mesh_component (preferably a cylinder):
                Visible: false
            transform_component
                - rotate to postion that you want the lever to be in when its on
            LeverHandle:
                mesh_component (preferably sphere):
                    Visible: false
                transform_component:
                    position at end of rod
        LeverPos1:
            mesh_component (preferably a cylinder):
                Visible: false
            transform_component:
                - rotate to postion that you want the lever to be in when its off
            LeverHandle:
                mesh_component (preferably sphere):
                    Visible: false
                transform_component:
                    - position at end of rod
        LeverRod:
            mesh_component (preferably a cylinder):
            LeverHandle:
                mesh_component (preferably sphere):
                transform_component:
                    - position at end of rod
                

sb_lane:
    RootEntity:
        - sb_lane_component
            MaybeInputDevice: set to an input_trigger_device set the interact input placed in the level
            PuckSpawnPoints: populate with "PuckSpawn1", "PuckSpawn2", and "PuckSpawn3"
        LeftWall (sb_wall prefab instance)
        RightWall (sb_wall prefab instance)
        DetectorZone1Points:
            - detector_zone_component
                Points: 1
            - mesh_component (preferably a cube)
            - transform_component
                LocalTransform: position closest to the entrance of the lane, and scale to be the largest
        PointsZone1Strip:
            -mesh_component (preferably a cube)
            - transform_component
                LocalTransform: position and scale along entire width on the ground in front of DetectorZone1Points
         DetectorZone1Points:
            - detector_zone_component
                Points: 2
            - mesh_component (preferably a cube)
            - transform_component
                LocalTransform: position between DetectorZone1Points and DetectorZone4Points, and scale to be the second largest
        PointsZone2Strip:
            - mesh_component (preferably a cube)
            - transform_component
                LocalTransform: position and scale along entire width on the ground in front of DetectorZone2Points
        DetectorZone1Points:
            - detector_zone_component
                Points: 4
            - mesh_component (preferably a cube)
            - transform_component
                LocalTransform: position at the very end of the lane, and scale to be the smallest
        PointsZone2Strip:
            - mesh_component (preferably a cube)
            - transform_component
                LocalTransform: position and scale along entire width on the ground in front of DetectorZone4Points
        LeverPlatform:
            - mesh_component (preferably cylinder or cube)
            - transform_component:
                LocalTransform: scale to be a flat platform and position at entrance of lane
        Lever (instance of sb_lever entity prefab):
            - transform_component:
                LocalTransform: position on top of LeverPlatform
        PuckSpawn1
            - transoform_component:
                LocalTransform: position in front of LeverPlatform in center of lane
        PuckSpawn2
            - transoform_component:
                LocalTransform: position in front of LeverPlatform on left side of lane
        PuckSpawn3
            - transoform_component:
                LocalTransform: position in front of LeverPlatform on right side of lane
            
#>

using { /Fortnite.com/Devices }
using { /Verse.org/SceneGraph }
using { /Verse.org/Simulation }
using { /Verse.org/SpatialMath }
using { /Fortnite.com/Game }
using { /Fortnite.com/Characters }
using { /Fortnite.com/Playspaces }
using { /Verse.org/Colors }
using {EntityPrefabs}


sb_lane_component<public> := class<final_super>(component):
    # ============ EDITABLE FIELDS =====================================
    @editable 
    LaneColor: color = NamedColors.CornflowerBlue

    @editable     
    var MaybeInputDevice: ?input_trigger_device = false # This input device assumes "require registered" is set in place of "add registered." It's default settings otherwise.

    @editable:
    PuckSpawnPoints: []entity = array{}

    # ============ EVENTS ===============================================
    # Signals when this component's OnEndSimulation function fires; used to cancel asyncronous logic when this component stops simulating
    EndSimulationEvent<internal>: event() = event(){}

    # Signaled when the score is updated from a zone, this event returns the change.
    # When a puck enters it's a positive change, and when a puck leaves it returns negative.
    ScoreUpdatedEvent<public>: event(int) = event(int){}

    # Signals when the MaybeRegisteredAgent value changes; passes an optional agent value which indicates which agent has been registered to this lane (if any)
    RegisteredAgentUpdateEvent: event(?agent) = event(?agent){}

    # Signals when a player has used all the pucks
    AllPucksUsedEvent: event() = event(){}

 
    # ============ INTERNAL VARS =======================================
    var MaybeRegisteredAgent: ?agent = false
    var MaybeLeverComponent: ?sb_lever_component = false
    var AllDetectorZones: []sb_detector_zone_component = array{} 
    var Score: int = 0

    var MaybeAgentRemovedSub: ?cancelable = false
    var MaybeFocusedPuck: ?sb_puck = false

    var PuckEntities : []sb_puck = array{}
    var PucksUsed : int = 0
    var LaneID : int = 0


    OnBeginSimulation<override>():void=
        (super:)OnBeginSimulation()
        set AllDetectorZones = Entity.FindDescendantComponentsArray(sb_detector_zone_component)
        # Find the only starting Lever component descendent 
        set MaybeLeverComponent = option. Entity.FindDescendantComponentsArray(sb_lever_component)[0]
        if (RoundMan := Entity.GetFortRoundManager[]):
            RoundMan.SubscribeRoundStarted(OnRoundStart)


    OnRoundStart()<suspends>:void=
        for(DetectorZoneComp : AllDetectorZones):
            # spawn. ListenForZoneUpdates(DetectorZoneComp)
            set DetectorZoneComp.MaybeLaneComponent = option. Self
        race:
            EndSimulationEvent.Await()
            block:  # On Lever Interacted With
                ListenForLeverInteract()
                Sleep(Inf)

    
    ListenForLeverInteract()<suspends>:void=
        if (LeverComponent := MaybeLeverComponent?):
            race:
                EndSimulationEvent.Await()
                loop:
                    Result := LeverComponent.AgentInteractedEvent.Await()
                    if (Result.IsReady?):
                        RegisterAgent(Result.Agent)
                    else:
                        UnregisterAgent()
                    Sleep(0.0)


    RegisterAgent(Agent: agent):void =
        if (InputDevice := MaybeInputDevice?) then InputDevice.Register(Agent)
        if (MaybeRegisteredAgent?) then UnregisterAgent()
        set MaybeRegisteredAgent = option. Agent
        if (Playspace := Entity.GetPlayspaceForEntity[]):
            Sub := Playspace.ParticipantRemovedEvent().Subscribe(OnAgentRemoved)
            set MaybeAgentRemovedSub = option. Sub
        RegisteredAgentUpdateEvent.Signal(MaybeRegisteredAgent)


    UnregisterAgent():void =
        if:
            InputDevice := MaybeInputDevice?
            Agent := MaybeRegisteredAgent?
        then:
            InputDevice.Unregister(Agent)
        set MaybeRegisteredAgent = false
        UnfocusPuck()
        if (AgentRemovedSub := MaybeAgentRemovedSub?):
            AgentRemovedSub.Cancel()
        RegisteredAgentUpdateEvent.Signal(MaybeRegisteredAgent)

    # Prevent the player from interacting with the lever and readying/unreading
    LockLever():void= 
        if (LeverComponent := MaybeLeverComponent?):
            LeverComponent.Lock()

    # Stop preventing the player from interacting with the lever and readying/unreading
    UnlockLever():void= 
        if (LeverComponent := MaybeLeverComponent?):
            LeverComponent.Unlock()


    OnAgentRemoved(Agent: agent):void =
        if (Agent = MaybeRegisteredAgent?):
            UnregisterAgent()


    SpawnPucks():void=
        for (PuckSpawnPoint : PuckSpawnPoints):
            MaybePuckEntity := SpawnPuckAtLocation(PuckSpawnPoint.GetGlobalTransform().Translation)
            if (PuckEntity := MaybePuckEntity?):
                set PuckEntities += array{PuckEntity}


    # Returns a sb_puck prefab entity if it successfully creates a puck at the given location
    # can't use decides or transacts because it needs to be able to add an entity ot the simulation entity and spawn a listener
    SpawnPuckAtLocation(Translation : vector3):?sb_puck=
        var MaybeNewPuck: ?sb_puck = option. sb_puck{}
        if:
            SimEntity := Entity.GetSimulationEntity[]
            NewPuck := MaybeNewPuck?
        then:
            SimEntity.AddEntities(array{NewPuck})
            # Place Puck at spawn location
            NewTransform := (/Verse.org/SpatialMath:) transform {
                Translation := Translation,
                Rotation := (/Verse.org/SpatialMath:) MakeRotationFromYawPitchRollDegrees(0.0, 0.0, 0.0),
                Scale := (/Verse.org/SpatialMath:) vector3{Forward := 1.0, Left := 1.0, Up := 1.0}}
            NewPuck.SetGlobalTransform(NewTransform)
            if(PuckComponent := NewPuck.GetComponent[sb_puck_component]):
                set PuckComponent.MaybeInputDevice = MaybeInputDevice
            spawn. ListenForPuckUsed(NewPuck)
        MaybeNewPuck


    ListenForZoneUpdates(DetectorZoneComp : sb_detector_zone_component)<suspends>:void=
        race:
            EndSimulationEvent.Await()
            loop: # Collect points
                ScoreChange := DetectorZoneComp.ZonePointUpdate.Await()
                set Score = Score + ScoreChange
                ScoreUpdatedEvent.Signal(Score)
                Sleep(0.0)


    ListenForPuckUsed(PuckEntity : sb_puck)<suspends>:void=
        if (PuckComp := PuckEntity.GetPuckComp[]):
            race:
                EndSimulationEvent.Await()
                block:
                    PuckComp.PuckLaunchedEvent.Await()

                    PuckComp.CameToAStopEvent.Await()
                    UnfocusPuck()
                    # Determine which zone the puck is in when it came to a full stop
                    if (HighestOverlappingZone := FindHighestScoringZone[PuckComp]):
                        set Score += HighestOverlappingZone.Points
                        Print("Scored {HighestOverlappingZone.Points}")
                        ScoreUpdatedEvent.Signal(Score)
                    else:
                        Print("No Points Scored :(")

                    set PucksUsed += 1
                    if (PucksUsed = PuckEntities.Length):
                        AllPucksUsedEvent.Signal()


    UnfocusPuck():void =
        if (FocusedPuck := MaybeFocusedPuck?):
            # First release the registered agent from the view
            if:
                Player := MaybeRegisteredAgent?
                FortChar := Player.GetFortCharacter[]
            then:
                FortChar.ReleaseFromStasis()
            
            set MaybeFocusedPuck = false


    FindHighestScoringZone(PuckComp: sb_puck_component)<decides><transacts>:sb_detector_zone_component = 
        PuckComp.IsInScene[]
        var MaybeHighestScoreDetector : ?sb_detector_zone_component = false
        # Get overlapping detector zones
        for:
            ThisZone : AllDetectorZones
            ThisZone.IsPuckOverlapping[PuckComp]
            (not MaybeHighestScoreDetector?) or (ThisZone.Points > MaybeHighestScoreDetector?.Points)
        do:
            set MaybeHighestScoreDetector = option. ThisZone
        
        if (HighestScoreDetector := MaybeHighestScoreDetector?):
            Print("HighestScoreScored: {HighestScoreDetector.Points}")
        else:
            Print("MaybeHighestScoreDetector is undefined")
        MaybeHighestScoreDetector?


    Reset():void=
        # Reset Game Info
        UnregisterAgent()
        RegisteredAgentUpdateEvent.Signal(MaybeRegisteredAgent)
        set Score = 0 
        ScoreUpdatedEvent.Signal(Score)
        # Reset Pucks
        for(Puck : PuckEntities):
            Puck.RemoveFromParent()
        set PuckEntities = array{}
        set PucksUsed = 0
        # Reset Lever
        UnlockLever()
        if (LeverComponent := MaybeLeverComponent?):
            LeverComponent.Reset()


# Prefab Helper Functions
(LanePF: sb_lane).GetLaneComp()<decides><transacts>:sb_lane_component = 
    LanePF.FindDescendantComponentsArray(sb_lane_component)[0]
