<#
@UUID: 71c39154-1d89-43e3-b6f7-5b1e57ee895b
@Task: Lever Component
@Description: An interactable lever that has a movement animation when interacted with. Used by plsyers to register to a lane and ready up for the game
@Tags: SceneGraph, PlayerInteraction, Movement
@Export: true
#>

<#
In-Editor Entity Prefab Setup:
    (NOTE: any setting not mentioned is assumed to be default or to your preference)

sb_lever
    RootEntity:
        - sb_lever_component
            MaybeLeverRod:  set to "LeverRod" in the editor
            MaybeLeverRodPos1: set to "LeverPos1" in the editor
            MaybeLeverRodPos2:   set to "LeverPos2" in the editor
        - sb_guage_microgame_component
        LeverInteraction
            - interaction_component
            - mesh_component (preferably cylinder)
                Visible: false
            - transform_component:
                LocalTransform: scale such that is covers the entire area where the lever rod could be positioned
        LeverBase1
            - mesh_component
        LeverBaswe2
            - mesh_component
        LeverPos1:
            mesh_component (preferably a cylinder):
                Visible: false
            transform_component
                - rotate to postion that you want the lever to be in when its on
            LeverHandle:
                mesh_component (preferably sphere):
                    Visible: false
                transform_component:
                    position at end of rod
        LeverPos1:
            mesh_component (preferably a cylinder):
                Visible: false
            transform_component:
                - rotate to postion that you want the lever to be in when its off
            LeverHandle:
                mesh_component (preferably sphere):
                    Visible: false
                transform_component:
                    - position at end of rod
        LeverRod:
            mesh_component (preferably a cylinder):
            LeverHandle:
                mesh_component (preferably sphere):
                transform_component:
                    - position at end of rod
                
#>

using { /Fortnite.com/Devices }
using { /Verse.org/SceneGraph }
using { /Verse.org/Simulation }
using { /Verse.org/SpatialMath }
using { /Fortnite.com/Game }

lever_interact_result := struct:
    Agent<public>: agent
    IsReady<public>: logic

sb_lever_component<public> := class<final_super>(component):
    # ============ EDITABLE FIELDS ==================================

    @editable_number(float):
        MinValue := option. 0.1
    MoveSpeed: float = 3.0

    @editable:
    MaybeLeverRod: ?entity = false

    @editable:
    MaybeLeverRodPos1: ?entity = false

    @editable:
    MaybeLeverRodPos2: ?entity = false

    # ============ EVENTS =======================================

    # Signals when this device's OnEndSimulation function is called; used primarily to cancel async logic after the game ends or the entity is removed from the scene.
    EndSimulationEvent<public>: event() = event(){}
    # Signals when an agent interacts with the lever
    AgentInteractedEvent: event(lever_interact_result) = event(lever_interact_result){}
    # Signals when the switch status has updated
    SwitchStatusUpdated: event() = event(){}

    # ============ INTERNAL VARS =======================================

    var IsReady : logic = false
    var IsLocked : logic = false


    OnEndSimulation<override>():void=
        (super:)OnEndSimulation()
        EndSimulationEvent.Signal()


    OnBeginSimulation<override>():void=
        (super:)OnBeginSimulation()

        if (RoundMan := Entity.GetFortRoundManager[]):
            RoundMan.SubscribeRoundStarted(OnRoundStart)


    OnRoundStart()<suspends>:void=
        if:
            LeverRod := MaybeLeverRod?
            LeverRodPos1 := MaybeLeverRodPos1?
        then:
            LeverRod.SetGlobalTransform(LeverRodPos1.GetGlobalTransform())

        race:
            EndSimulationEvent.Await()
            block:
                if (Interactable := Entity.FindDescendantComponentsArray(interactable_component)[0]):
                    # Listen for agent interaction
                    loop: 
                        Interactable.Enable()
                        Agent := Interactable.SucceededEvent.Await()
                        if(IsReady?) then Reset()
                        else TurnOn()
                        AgentInteractedEvent.Signal(lever_interact_result{Agent:=Agent, IsReady:=IsReady})
                        Sleep(0.0)
        

    Reset():void=
        set IsReady = false
        SwitchStatusUpdated.Signal()
        spawn. LerpRotation(false)


    TurnOn():void=
        set IsReady = true
        SwitchStatusUpdated.Signal()
        spawn. LerpRotation(true)


    Lock():void=
        set IsLocked = true
        if (Interactable := Entity.FindDescendantComponentsArray(interactable_component)[0]):
            Interactable.Disable()


    Unlock():void=
        set IsLocked = false
        if (Interactable := Entity.FindDescendantComponentsArray(interactable_component)[0]):
            Interactable.Enable()


    LerpRotation(IsFlippingOn: logic)<suspends>:void=
        if (Interactable := Entity.FindDescendantComponentsArray(interactable_component)[0]):
            Interactable.Disable()
        race:
            EndSimulationEvent.Await()
            SwitchStatusUpdated.Await()
            block:
                if:
                    StartRotation := MaybeLeverRodPos1?.GetGlobalTransform().Rotation #OriginRotation * EndGlobalRotation
                    EndRotation := MaybeLeverRodPos2?.GetGlobalTransform().Rotation
                then:
                    Timestamp := GetSimulationElapsedTime()
                    loop:
                        Sleep(0.0)
                        if(LeverRod := MaybeLeverRod?):
                            DeltaTime := GetSimulationElapsedTime() - Timestamp
                            Interp := Clamp(MoveSpeed*DeltaTime, 0.0, 1.0)
                            var NewRotation : rotation = rotation{}

                            if(IsFlippingOn?):
                                set NewRotation = Slerp(StartRotation, EndRotation, Interp)
                            else:
                                set NewRotation = Slerp(EndRotation, StartRotation, Interp)

                            NewTransform := (/Verse.org/SpatialMath:) transform{
                                Translation:=LeverRod.GetGlobalTransform().Translation, 
                                Rotation:=NewRotation, 
                                Scale:=LeverRod.GetGlobalTransform().Scale
                            }
                            LeverRod.SetGlobalTransform(NewTransform)
                            if(Interp >= 1.0):
                                if:
                                    Interactable := Entity.FindDescendantComponentsArray(interactable_component)[0]
                                    not IsLocked?
                                then:
                                    Interactable.Enable()
                                break
                        else:
                            Print("Lever Component can't find lever rod")
                            break