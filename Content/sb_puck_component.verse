<#
@UUID: 71c39154-1d89-43e3-b6f7-5b1e57ee895b
@Task: Shuffle Board Puck Component
@Description: A puck that is launched horizontally when interacted with by a player. The direction of the launch is based on the interacting agent's position in relation to the puck, and its speed is determined by a timing minigame. The puck can also bounce off walls.
@Tags: SceneGraph, Gameplay, PlayerInteraction, Physics, Movement
@Export: true
#>

<#
In-Editor Entity Prefab Setup:
    (NOTE: any setting not mentioned is assumed to be default or to your preference)

sb_puck
    RootEntity:
        - sb_puck_component
        - sb_guage_microgame_component
        PuckCollision
            - interactable_component
            - mesh_component (preferably cylinder):
            - transform_component
                - LocalTransform: scale the cylinder into the shape of a large flat puck
                
#>

using { /Verse.org }
using { /Verse.org/Concurrency }
using { /Verse.org/Simulation }
using { /Verse.org/Native }
using { /Verse.org/SceneGraph }
using { /Verse.org/SpatialMath }
using { /Fortnite.com/Game }
using { /Fortnite.com/Playspaces }
using { /Fortnite.com/Characters }
using { /Fortnite.com/Devices }
using { /UnrealEngine.com/Temporary/SpatialMath }
using { /UnrealEngine.com/Temporary/Diagnostics }
using {EntityPrefabs}

sb_puck_debug := class(debug_draw_channel) {}

sb_puck_component<public> := class<final_super>(component):

    # Percentage of speed that is lost per second of movement
    @editable_slider(float):
        MinValue := option. 0.0
        MaxValue := option. 1.0
    var Friction: float = 0.4
    
    # Percentage of speed that is retained after colliding with a surface
    @editable_slider(float):
        MinValue := option. 0.0
        MaxValue := option. 1.0
    var Bounciness: float = 0.7
    
    # The smallest possible speed value the puck can be launched at
    @editable_number(float):
        MinValue := option. 0.0
    var MinInitialSpeed: float = 100.0
    
    # The greatest possible speed value the puck can be launched at
    @editable_number(float):
        MinValue := option. 0.0
    var MaxInitialSpeed: float = 1000.0

    # Puck is stopped after falling at or bellow this speed value
    @editable_number(float):
        MinValue := option. 0.0
    var StopThreshold: float = 20.0


    # Signals when this puck is Launched
    PuckLaunchedEvent<internal>: event() = event(){}
    CameToAStopEvent<public>: event() = event(){}
    EndSimulationEvent<public>: event() = event(){}

    var MaybeInteractable: ?interactable_component = false
    var MaybeGaugeComp: ?sb_gauge_microgame_component = false
    var LastZoneEntered : ?sb_detector_zone_component = false

    var MaybePreTickHandler: ?cancelable = false
    var CurrentVelocity: (/Verse.org/SpatialMath:) vector3 = (/Verse.org/SpatialMath:) vector3 {}
    var MaybeInputDevice : ?input_trigger_device = false


    DebugDraw:debug_draw = debug_draw{Channel := sb_puck_debug}

    OnBeginSimulation<override>():void=
        (super:)OnBeginSimulation()

        #set MaybeCamFocusComp = option. Entity.FindDescendantComponentsArray(gameplay_camera_focus_component)[0]      
        if(PuckRootComp := Entity.GetComponent[sb_gauge_microgame_component]):
            set MaybeGaugeComp = option. PuckRootComp

        if (Interactable := Entity.FindDescendantComponentsArray(interactable_component)[0]):
            set MaybeInteractable = option. Interactable
            Interactable.SucceededEvent.Subscribe(OnInteracted)

        if (RoundMan := Entity.GetFortRoundManager[]):
            RoundMan.SubscribeRoundStarted(OnRoundStart)


    OnInteracted(Instigator: agent):void =
        if (Player := player[Instigator]):
            spawn. OnPlayerInteract(Player)
            

    OnPlayerInteract(Player: player)<suspends>: void =
        if:
            Playspace := Entity.GetPlayspaceForEntity[]
            Character := Player.GetFortCharacter[]
            GaugeComp := MaybeGaugeComp?
        then:
            # Disable interactable comp
            if (Interactable := MaybeInteractable?):
                Interactable.Disable()

            race:
                Character.EliminatedEvent().Await()
                EndSimulationEvent.Await()
                loop:
                    PlayerThatLeft := Playspace.PlayerRemovedEvent().Await()
                    if (Player = PlayerThatLeft):
                        break
                block:
                    if(InputDevice := MaybeInputDevice?):
                        # Wait Until the gauge minigame is complete
                        MaybeFillPercent := GaugeComp.PlayMicrogame(Player, InputDevice)
                        if (FillPercent := MaybeFillPercent?):
                            # Determine magnitude of the impulse
                            ShotMagnitude := Lerp(MinInitialSpeed,MaxInitialSpeed,FillPercent)

                            # Find the Direction based on the player's position compared to the puck's
                            EntityHorizontalPos := Entity.GetGlobalTransform().Translation.RemoveUp()
                            PlayerHorizontalPos := FromVector3(Character.GetTransform().Translation).RemoveUp()
                            ShotDirection: (/Verse.org/SpatialMath:) vector3  = (EntityHorizontalPos - PlayerHorizontalPos).MakeUnitVector()

                            # Calculate the velocity from the direction and magnitude
                            ImpulseVelocity := ShotDirection * ShotMagnitude
                            ApplyImpulse(ImpulseVelocity)
                            PuckLaunchedEvent.Signal()

                            # Draw the direciton of the shot
                            DebugDraw.DrawArrow(
                                Entity.GetGlobalTransform().Translation,ImpulseVelocity,
                                ?DrawDurationPolicy := debug_draw_duration_policy.FiniteDuration,
                                ?Duration := 5.0)
                            Sleep(0.0)
                            return

            # Should only reach this part if the gauge microgame is prematurely canceled
            if (Interactable := MaybeInteractable?):
                Interactable.Enable()
    

    OnEndSimulation<override>():void=
        (super:)OnEndSimulation()
        if. MaybePreTickHandler?.Cancel()
        EndSimulationEvent.Signal()
    

    OnRoundStart()<suspends>:void =
        # Subscribe to post-physics tick events
        set MaybePreTickHandler = option. TickEvents.PrePhysics.Subscribe(OnPrePhysics)

    
    OnPrePhysics(Delta: float):void = 
        if:
            IsSimulating[] and HorizontalSpeed() > StopThreshold
            InteractableComp := MaybeInteractable?
            ColliderEntity  := InteractableComp.Entity
        then:
            # Apply Friction
            set CurrentVelocity =  CurrentVelocity - CurrentVelocity*Friction*Delta

            # Check if puck has come to a stop
            if (CurrentVelocity.RemoveUp().IsAlmostZero[StopThreshold]):
                set CurrentVelocity = (/Verse.org/SpatialMath:) vector3{}
                CameToAStopEvent.Signal()
            
            # Sweep For Collisions
            if (Collider := InteractableComp.Entity.GetComponent[mesh_component]):
                for:
                    SweepHit : ColliderEntity.FindSweepHits(CurrentVelocity, ColliderEntity.GetGlobalTransform())
                    MeshComp := mesh_component[SweepHit.TargetComponent]
                    MeshRootEntity := MeshComp.Entity.GetParent[]
                    WallEntity := EntityPrefabs.sb_wall[MeshRootEntity]
                do:
                    ContactNormal := SweepHit.ContactFaceNormal.RemoveUp()
                    set CurrentVelocity = ReflectVector(CurrentVelocity, ContactNormal) * Bounciness
                    
            # Apply Global Translation
            OldGlobal := Entity.GetGlobalTransform()
            NewGlobal :=  (/Verse.org/SpatialMath:) transform{
                Translation := OldGlobal.Translation + CurrentVelocity*Delta,
                Rotation := OldGlobal.Rotation,
                Scale := OldGlobal.Scale
            }
            Entity.SetGlobalTransform(NewGlobal)


    HorizontalSpeed()<reads>: float =
        CurrentVelocity.LengthForwardLeft()

    
    ApplyImpulse(ImpulseVelocity: (/Verse.org/SpatialMath:) vector3): void=
        set CurrentVelocity = CurrentVelocity + ImpulseVelocity

    
(PuckPF: sb_puck).GetPuckComp()<decides><transacts>:sb_puck_component =
    PuckPF.GetComponent[sb_puck_component]



