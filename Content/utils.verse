<#
@UUID: 71c39154-1d89-43e3-b6f7-5b1e57ee895b
@Task: Shuffle Board Minigame Utils
@Description: Helper functions used to convert data types, traverse scene graph entity trees, and  operate on data structures
@Tags: Utility, SceneGraph, DataStructures, DataConversion
@Export: true
#>

using { /Verse.org/SceneGraph }
using { /Verse.org/SpatialMath }
using { /Verse.org/Simulation }
using { /UnrealEngine.com/Temporary/SpatialMath }


###################### Start Primitive Data Conversion ####################################################################

(Val:logic).ToIntSign():int = if. Val? then 1 else -1
(Val:logic).ToFloatSign():float = if. Val? then 1.0 else -1.0

###################### End Primitive Data Conversion #######################################################################

# identical to Sleep(0.0) except that it returns the delta time between frames
SleepFrame()<suspends>: float=
    Timestamp := GetSimulationElapsedTime()
    Sleep(0.0)
    GetSimulationElapsedTime() - Timestamp


# returns the vector with the vertical component (Up) removed
(Vector: (/Verse.org/SpatialMath:) vector3).RemoveUp<public>()<computes>:(/Verse.org/SpatialMath:) vector3 =
    (/Verse.org/SpatialMath:) vector3 {
        Forward := Vector.Forward,
        Up := 0.0,
        Left := Vector.Left
    }

# Helper function to return an array (instead of a generator) of all components attached to descendant entities to and including `InEntity` of type `component_type`.
(Entity:entity).FindDescendantComponentsArray<public>(ComponentType:castable_subtype(component))<transacts>:[]ComponentType =
    for (Comp : Entity.FindDescendantComponents(ComponentType)) do Comp


# Removes an element from the given map and returns a new map without that element
(Map:[keytype]valuetype where keytype:subtype(comparable), valuetype:type).RemoveKey(ElementToRemove:keytype)<transacts>:[keytype]valuetype=
    var NewMap:[keytype]valuetype = map{}
    for (Key -> Value : Map, Key <> ElementToRemove):
        set NewMap = ConcatenateMaps(NewMap, map{Key => Value})
    return NewMap

StringToMessage<localizes>(InString : string) : message = "{InString}"